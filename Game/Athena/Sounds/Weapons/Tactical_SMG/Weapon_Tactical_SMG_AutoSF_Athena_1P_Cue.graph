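# First-person fire sound cue for the Athena Tactical SMG, named as its AutoSF variant, with an added low-ammo warning layer.
# A mixer combines one of four random fire samples, using close-projectile attenuation and the 1P fire sound class, with a modulated warning branch.
# That branch plays a high rifle low-ammo sample only when LowAmmo is true, under its own critical sound class.
# The cue uses first-person gunfire concurrency and disables virtualization.
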SoundCue /Game/Athena/Sounds/Weapons/Tactical_SMG/Weapon_Tactical_SMG_AutoSF_Athena_1P_Cue
root=Mixer_0
nodes [
 {id=Attenuation_118, children [Random_1], properties {attenuationSettings=&Fort_Weapon_Projectile_Close_Athena_Att}}
 {id=Branch_3, children [WavePlayer_0,null,null], properties {boolParameterName=LowAmmo}}
 {id=Mixer_0, children [SoundClass_77, SoundClass_0], properties {inputVolume [1,1]}}
 {id=Modulator_4, children [Branch_3], properties {pitchMax=1, volumeMax=1.15, volumeMin=1}}
 {
  id=Random_1
  children [WavePlayer_1,WavePlayer_5,WavePlayer_6,WavePlayer_7]
  properties {bShouldExcludeFromBranchCulling=true, weights [1,1,1,1]}
 }
 {
  id=SoundClass_0
  children [Modulator_4]
  properties {soundClassOverride=/Game/Sounds/FortSoundClassesAndModes/Fort_WeaponRanged_LowAmmo_Critical_PSM}
 }
 {
  id=SoundClass_77
  children [Attenuation_118]
  properties {soundClassOverride=&Fort_WeaponRanged_Fire_1P_PSM}
 }
 {id=WavePlayer_0, properties {soundWaveAssetPtr=/Game/Sounds/Fort_Weapons/LowAmmo/weapon_low_ammo_hi_rifle_05}}
 {id=WavePlayer_1, properties {soundWaveAssetPtr=/Game/Athena/Sounds/Weapons/Tactical_SMG/Rev/TacSMG_Fire_1p_01}}
 {id=WavePlayer_5, properties {soundWaveAssetPtr=/Game/Athena/Sounds/Weapons/Tactical_SMG/Rev/TacSMG_Fire_1p_02}}
 {id=WavePlayer_6, properties {soundWaveAssetPtr=/Game/Athena/Sounds/Weapons/Tactical_SMG/Rev/TacSMG_Fire_1p_03}}
 {id=WavePlayer_7, properties {soundWaveAssetPtr=/Game/Athena/Sounds/Weapons/Tactical_SMG/Rev/TacSMG_Fire_1p_04}}
]
properties {
 attenuationSettings=&Fort_Weapon_Projectile_FirstPerson_Athena_Att
 bDebug=true
 maxDistance=2097152
 soundClassObject=&Fort_WeaponRanged_Fire_1P_PSM
 virtualizationMode=Disabled
 volumeMultiplier=0.7
 concurrencyOverrides {bLimitToOwner=true, maxCount=2, resolutionRule=StopOldest, volumeScale=0.8667}
 concurrencySet [&Weapon_Gunfire_FirstPerson_Conc]
}
