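# Sound cue for Tactical SMG gunfire in Athena, layering close, medium and distant shots through distance-specific attenuations.
# The close layer plays randomized 1P-named waves under a third-person fire sound class.
# Randomized medium and distant layers, with modulated volume and pitch, are submixed under a distant-fire sound class.
# An owner-limited, stop-oldest concurrency override accompanies global and local gunfire concurrency sets.
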SoundCue /Game/Athena/Sounds/Weapons/Tactical_SMG/Weapon_Tactical_SMG_AutoSF_Athena_Cue
root=Mixer_1
nodes [
 {
  id=Attenuation_118
  children [Modulator_0]
  properties {attenuationSettings=&Fort_Weapon_Projectile_Close_Athena_Att}
 }
 {
  id=Attenuation_119
  children [Modulator_2]
  properties {attenuationSettings=&Fort_Weapon_Projectile_Medium_Mono_Athena_Att}
 }
 {
  id=Attenuation_120
  children [Modulator_1]
  properties {attenuationSettings=&Fort_Weapon_Projectile_Distant_Mono_Athena_Att}
 }
 {id=Mixer_1, children [SoundClass_77, SoundClass_76], properties {inputVolume [0.9,0.9]}}
 {id=Mixer_2, children [Attenuation_119, Attenuation_120], properties {inputVolume [0.35,0.2]}}
 {id=Modulator_0, children [Random_1], properties {volumeMax=0.6, volumeMin=0.6}}
 {id=Modulator_1, children [Random_3], properties {pitchMin=1, volumeMax=1.3, volumeMin=1.2}}
 {id=Modulator_2, children [Random_2], properties {pitchMax=1.2, pitchMin=1.1, volumeMax=2, volumeMin=2}}
 {
  id=Random_1
  children [WavePlayer_5,WavePlayer_7,WavePlayer_8,WavePlayer_9]
  properties {bShouldExcludeFromBranchCulling=true, weights [1,1,1,1]}
 }
 {
  id=Random_2
  children [WavePlayer_10, WavePlayer_11, WavePlayer_12, WavePlayer_13]
  properties {bShouldExcludeFromBranchCulling=true, weights [1,1,1,1]}
 }
 {id=Random_3, children [WavePlayer_14, WavePlayer_15, WavePlayer_16], properties {bShouldExcludeFromBranchCulling=true, weights [1,1,1]}}
 {id=SoundClass_76, children [Mixer_2], properties {soundClassOverride=&Fort_WeaponRanged_Fire_Distant}}
 {
  id=SoundClass_77
  children [Attenuation_118]
  properties {soundClassOverride=&Fort_WeaponRanged_Fire_3P_PSM}
 }
 {id=WavePlayer_10, properties {soundWaveAssetPtr=/Game/Athena/Sounds/Weapons/Tactical_SMG/Rev/TacSMG_Fire_Medium_01}}
 {id=WavePlayer_11, properties {soundWaveAssetPtr=/Game/Athena/Sounds/Weapons/Tactical_SMG/Rev/TacSMG_Fire_Medium_02}}
 {id=WavePlayer_12, properties {soundWaveAssetPtr=/Game/Athena/Sounds/Weapons/Tactical_SMG/Rev/TacSMG_Fire_Medium_03}}
 {id=WavePlayer_13, properties {soundWaveAssetPtr=/Game/Athena/Sounds/Weapons/Tactical_SMG/Rev/TacSMG_Fire_Medium_04}}
 {id=WavePlayer_14, properties {soundWaveAssetPtr=/Game/Athena/Sounds/Weapons/Tactical_SMG/Rev/TacSMG_Fire_Distant_01}}
 {id=WavePlayer_15, properties {soundWaveAssetPtr=/Game/Athena/Sounds/Weapons/Tactical_SMG/Rev/TacSMG_Fire_Distant_02}}
 {id=WavePlayer_16, properties {soundWaveAssetPtr=/Game/Athena/Sounds/Weapons/Tactical_SMG/Rev/TacSMG_Fire_Distant_03}}
 {id=WavePlayer_5, properties {soundWaveAssetPtr=/Game/Athena/Sounds/Weapons/Tactical_SMG/Rev/TacSMG_Fire_1p_01}}
 {id=WavePlayer_7, properties {soundWaveAssetPtr=/Game/Athena/Sounds/Weapons/Tactical_SMG/Rev/TacSMG_Fire_1p_02}}
 {id=WavePlayer_8, properties {soundWaveAssetPtr=/Game/Athena/Sounds/Weapons/Tactical_SMG/Rev/TacSMG_Fire_1p_03}}
 {id=WavePlayer_9, properties {soundWaveAssetPtr=/Game/Athena/Sounds/Weapons/Tactical_SMG/Rev/TacSMG_Fire_1p_04}}
]
properties {
 bDebug=true
 maxDistance=27500
 volumeMultiplier=1
 concurrencyOverrides {bLimitToOwner=true, maxCount=3, resolutionRule=StopOldest}
 concurrencySet [&Weapon_Gunfire_Global_Conc, &Weapon_Gunfire_Local_Conc]
}
